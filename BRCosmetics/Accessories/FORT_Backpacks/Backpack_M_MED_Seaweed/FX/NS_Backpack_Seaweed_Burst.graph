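# Niagara burst effect for the Seaweed backpack cosmetic that flashes short-lived flame jets from two points on the backpack's skeletal mesh.
# A single BurstFlames emitter spawns two particles on the first loop, each assigned a spawn group that picks its bone or socket to follow.
# Four mesh renderers draw the flames, with per-layer color, scale and orientation variables animated by curves over lifetime.
# A light renderer adds a red glow that fades in and out.
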NiagaraSystem /BRCosmetics/Accessories/FORT_Backpacks/Backpack_M_MED_Seaweed/FX/NS_Backpack_Seaweed_Burst
dependencies {
 usedDataInterfaces [&NiagaraDataInterfaceColorCurve, &NiagaraDataInterfaceVectorCurve]
 usedDynamicInputs []
 usedModules [&SystemState]
}
emitters {BurstFlames {
  emitterSpawnScript [&FN_PreserveSkydiveUserVariables() {module=PreserveUserVariables}]
  emitterUpdateScript [
   &EmitterState(Life Cycle Mode=::NewEnumerator1, Loop Behavior=::NewEnumerator1, Loop Duration=1, UseLoopDelay=false)
   &SpawnBurst_Instantaneous(Spawn Count=2, Spawn Probability=HlslExpression() {hlslExpression="Emitter.LoopCount==0"}, Use Spawn Probability=true)
  ]
  particleSpawnScript [
   /Game/Effects/Niagara/Modules/Groups/FN_SetSpawnGroup_v1/FN_SetSpawnGroup(Emitter Bone Count=2)
   /Niagara/Modules/Spawn/Initialization/InitializeParticle(
    Color=<rgba 1 1 1 0.1>
    Lifetime=0.75
    Mesh Scale=<0.075 0.075 0.0875>
    Module.Write Color=true
    Module.Write Lifetime=true
    Module.Write Mass=false
    Module.Write Material Random=true
    Module.Write Position=false
    Module.Write Scale=true
    Module.Write SpriteRotation=false
    Module.Write SpriteSize=false
   )
   /SetVariables(Particles.FlameCoreColor:LinearColor=<rgba 2 0.3392 0.4442 0.1>, Particles.FlameCoreScale:Vector3f=<0.25 0.05 0.05>) {isSetParameters=true}
   /SetVariables(Particles.InnerFlameCoreColor:LinearColor=<rgba 10 0 0 0.025>, Particles.InnerFlameCoreScale:Vector3f=<0.25 0.2 0.2>) {isSetParameters=true}
   /SetVariables(Particles.BrightCoreColor:LinearColor=<rgba 2.5 0.5078 0.6338 0.1>, Particles.BrightCoreScale:Vector3f=<0.15 0.15 0.15>) {isSetParameters=true}
   &DynamicMaterialParameters(
    Index 0 Param 2=2
    Index 0 Param 3=20
    Module.Param0WriteEnabled=true
    Module.Param1WriteEnabled=true
    Module.Param2WriteEnabled=true
    Module.Param3WriteEnabled=true
   )
   /SetVariables(Particles.LightsExponent=4, Particles.LightsRadius=30) {isSetParameters=true}
  ]
  particleUpdateScript [
   &ParticleState()
   &SkeletalMeshLocation(
    Apply Sampled Orientation to Mesh Particles=true
    Bone / Socket Index=$Particles.SpawnGroup
    Inherit Velocity=false
    MaskSpawnGroup=false
    Mesh Sampling Type=::NewEnumerator5
    Sampled Orientation Coordinate Space=::World
    Sampled Position Coordinate Space=::World
    Sampled Position Offset Coordinate Space:ENiagaraMeshTransforms=::NewEnumerator1
    Sampled Velocity Coordinate Space:ENiagaraCoordinateSpace=::World
    Skeletal Mesh=$User.SkeletalMesh
    Socket Sampling Mode=::NewEnumerator5
   )
   /SetVariables(Particles.BrightCoreColor:LinearColor=&MultiplyLinearColor(
     Linear Color=$Particles.Initial.BrightCoreColor
     Scale Factor=&MakeVector4_FromLinearColor(LinearColor=&ColorFromCurve(ColorCurve=DataInterface() {propertyValues {
         AlphaCurve {keys [<at 0 1>]}
         BlueCurve {keys [<at 0 20 RCIM_Cubic>, {arriveTangent=-25.3333, interpMode=RCIM_Cubic, leaveTangent=-25.3333, time=0.25, value=4}, <at 0.75 1>]}
         GreenCurve {keys [<at 0 20 RCIM_Cubic>, {arriveTangent=-25.3333, interpMode=RCIM_Cubic, leaveTangent=-25.3333, time=0.25, value=4}, <at 0.75 1>]}
         RedCurve {keys [<at 0 20 RCIM_Cubic>, {arriveTangent=-25.3333, interpMode=RCIM_Cubic, leaveTangent=-25.3333, time=0.25, value=4}, <at 0.75 1>]}
        }}))
    )) {isSetParameters=true}
   /SetVariables(Particles.FlameCoreOrientation:Quat4f=&MultiplyQuaternion(
     Quaternion A=$Particles.MeshOrientation
     Quaternion B=&FindQuatBetweenTwoVectors(From Vector=<-1 0 0>, To Vector=<1 0 0>)
    )) {isSetParameters=true}
   /SetVariables(Particles.OuterFlameCoreOrientation:Quat4f=&MultiplyQuaternion(
     Quaternion A=$Particles.MeshOrientation
     Quaternion B=&FindQuatBetweenTwoVectors(From Vector=<0 0 1>, To Vector=<1 0 0>)
    )) {isSetParameters=true}
   &ScaleMeshSize(Scale Factor=&VectorFromCurve(VectorCurve=DataInterface() {propertyValues {
       XCurve {keys [<at 0 0.25>, {arriveTangent=0.9375, interpMode=RCIM_Cubic, leaveTangent=0.9375, time=0.1, value=1}, <at 0.8 1>, <at 1 0.25>]}
       YCurve {keys [<at 0 0.25>, {arriveTangent=0.9375, interpMode=RCIM_Cubic, leaveTangent=0.9375, time=0.1, value=1}, <at 0.8 1>, <at 1 0.25>]}
       ZCurve {keys [
         <at 0 0 RCIM_Cubic>
         {arriveTangent=1.3333, interpMode=RCIM_Cubic, leaveTangent=1.3333, time=0.1, value=1.5}
         {arriveTangent=-1.6667, interpMode=RCIM_Cubic, leaveTangent=-1.6667, time=0.75, value=1}
         <at 1 0 RCIM_Cubic>
        ]}
      }}))
   /SetVariables(Particles.FlameCoreScale:Vector3f=&Multiply_Vector(
     A=$Particles.Initial.FlameCoreScale
     B=&VectorFromCurve(VectorCurve=DataInterface() {propertyValues {
        XCurve {keys [<at 0 0 RCIM_Cubic>, <at 0.1 1.5>, {arriveTangent=-1.6667, interpMode=RCIM_Cubic, leaveTangent=-1.6667, time=0.75, value=1}, <at 1 0>]}
        YCurve {keys [<at 0 0.25 RCIM_Cubic>,<at 0.1 1>,<at 0.8 1>,<at 1 0.25>]}
        ZCurve {keys [<at 0 0.25 RCIM_Cubic>,<at 0.1 1>,<at 0.8 1>,<at 1 0.25 RCIM_Cubic>]}
       }})
    )) {isSetParameters=true}
   /SetVariables(Particles.InnerFlameCoreScale:Vector3f=&Multiply_Vector(
     A=$Particles.Initial.InnerFlameCoreScale
     B=&VectorFromCurve(VectorCurve=DataInterface() {propertyValues {
        XCurve {keys [<at 0 0 RCIM_Cubic>, <at 0.1 1.5>, {arriveTangent=-1.6667, interpMode=RCIM_Cubic, leaveTangent=-1.6667, time=0.75, value=1}, <at 1 0>]}
        YCurve {keys [<at 0 0.25 RCIM_Cubic>,<at 0.1 1>,<at 0.8 1>,<at 1 0.25>]}
        ZCurve {keys [<at 0 0.25 RCIM_Cubic>,<at 0.1 1>,<at 0.8 1>,<at 1 0.25 RCIM_Cubic>]}
       }})
    )) {isSetParameters=true}
   /SetVariables(Particles.BrightCoreScale:Vector3f=&Multiply_Vector(
     A=$Particles.Initial.BrightCoreScale
     B=&VectorFromCurve(VectorCurve=DataInterface() {propertyValues {
        XCurve {keys [<at 0 0 RCIM_Cubic>, <at 0.1 1.5>, {arriveTangent=-1.6667, interpMode=RCIM_Cubic, leaveTangent=-1.6667, time=0.75, value=1}, <at 1 0>]}
        YCurve {keys [<at 0 0.25 RCIM_Cubic>,<at 0.1 1>,<at 0.8 1>,<at 1 0.25>]}
        ZCurve {keys [<at 0 0.25 RCIM_Cubic>,<at 0.1 1>,<at 0.8 1>,<at 1 0.25 RCIM_Cubic>]}
       }})
    )) {isSetParameters=true}
   /SetVariables(Particles.Light Color:LinearColor=&ColorFromCurve(ColorCurve=DataInterface() {propertyValues {
       AlphaCurve {keys [<at 0 0>, {arriveTangent=1.3333, interpMode=RCIM_Cubic, leaveTangent=1.3333, time=0.2499, value=1}, <at 0.75 1 RCIM_Cubic>, <at 1 0>]}
       BlueCurve {keys [<at 0.25 0>,<at 0.5 0>,<at 0.75 0>]}
       GreenCurve {keys [<at 0.25 0>,<at 0.5 0>,<at 0.75 0>]}
       RedCurve {keys [<at 0.25 2.5>,<at 0.5 5>,<at 0.75 2.5>]}
      }})) {isSetParameters=true}
   /SetVariables(Particles.LightPosition:Vector3f=&Add_Vector(A=Linked() {linkedVariable {name=Particles.Position, type=NiagaraPosition}}, B=<0 0 -8>)) {isSetParameters=true}
  ]
  rendererClasses [&NiagaraMeshRendererProperties, &NiagaraLightRendererProperties]
  renderers [
   {rendererClass=&NiagaraMeshRendererProperties, rendererIndex=0}
   {rendererClass=&NiagaraMeshRendererProperties, rendererIndex=1}
   {rendererClass=&NiagaraMeshRendererProperties, rendererIndex=2}
   {rendererClass=&NiagaraMeshRendererProperties, rendererIndex=3}
   {rendererClass=&NiagaraLightRendererProperties, rendererIndex=4}
  ]
 }}
summary {
 systemName=NS_Backpack_Seaweed_Burst
 emitters [{emitterName=BurstFlames, rendererClasses [&NiagaraMeshRendererProperties, &NiagaraLightRendererProperties]}]
 userVariables [
  User.bIsFrontEndPreview {type=NiagaraBool}
  User.PlayerSpeed {type=NiagaraFloat, defaultValue=0, description=""}
  User.SkeletalMesh {
   type=NiagaraDataInterfaceSkeletalMesh
   defaultValue {
    struct=NiagaraExt_VariableValue_DataInterface
    value {dataInterface=&self:NiagaraDataInterfaceSkeletalMesh_0}
   }
  }
 ]
}
